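; Quest variant of a state-driven building gameplay actor that can destroy itself when a linked world actor is destroyed.
; If DestroyWithLinkedActor is set, BeginPlay waits five seconds, then runs a multi-object line trace between two offsets from its location.
; Hits of LinkedActorClass are stored as BoundActor and have OnDestroyed bound to destroy this actor, with debug prints.
; EndPlay unbinds OnDestroyed from a valid BoundActor.
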(event EventBeginPlay
  (|CallParentFunction)
  (if (Variables|LinkedActor|GetDestroyWithLinkedActor)
    (Utilities|FlowControl|Delay 5.0)
    (Development|PrintString "Starting Trace For Linked Actor" true true "(R=0.000000,G=0.660000,B=1.000000,A=1.000000)" 4.0)
    (bind _outhits (Collision|MultiLineTraceForObjects (+ (Transformation|GetActorLocation) (Variables|LinkedActor|GetStartPositionOffsetForTrace)) (+ (Transformation|GetActorLocation) (Variables|LinkedActor|GetEndPositionOffsetForTrace)) (Utilities|Array|MakeArray "ObjectTypeQuery1")))
    (bind _array_element (Utilities|Array|ForEachLoopwithBreak _outhits)
      (:LoopBody
        (bind (_array_element_bblockinghit _array_element_binitialoverlap _array_element_time _array_element_distance _array_element_location _array_element_impactpoint _array_element_normal _array_element_impactnormal _array_element_physmat _array_element_hitactor) (Collision|BreakHitResult _array_element))
        (if (Utilities|IsA _array_element_hitactor (Variables|LinkedActor|GetLinkedActorClass))
          (bind _output_get (Variables|LinkedActor|SetBoundActor _array_element_hitactor))
          (Game|BindEventtoOnDestroyed _output_get (delegate Custom|CustomEvent_1))
          (Development|PrintString (Fortnite|GetPlayerorObjectNameSafe _output_get) true true "(R=0.000000,G=0.660000,B=1.000000,A=1.000000)" 45.0)))
      (:Completed
        _))
    (Utilities|IsValid (Variables|LinkedActor|GetBoundActor)
      (:Is%20Not%20Valid))))

(event Custom|CustomEvent_1
  (params (DestroyedActor :type Actor%20Object%20Reference))
  (Actor|DestroyActor))

(event Collision|EventActorBeginOverlap
  (params (OtherActor :type Actor%20Object%20Reference)))

(event EventTick
  (params (DeltaSeconds :type Float%20%28single-precision%29)))

(event EventEndPlay
  (params (EndPlayReason :type EEndPlayReason%20Enum :default "Destroyed"))
  (bind _boundactor (Variables|LinkedActor|GetBoundActor))
  (Utilities|IsValid _boundactor
    (:Is%20Valid
      (Game|UnbindallEventsfromOnDestroyed _boundactor))))

(fn ConstructionScript
  (|CallParentFunction))
